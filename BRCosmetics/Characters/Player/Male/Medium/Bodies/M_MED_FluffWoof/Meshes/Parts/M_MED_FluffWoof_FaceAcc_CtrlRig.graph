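# Control Rig for the FluffWoof medium male FaceAcc part, posing its tie and left-thigh scissors, comb and strap mount.
# Clavicle deform bones follow their sources; NotRBAN-weighted proxy constraints aim the tie chain, with a DeltaRotation spine-bend correction.
# An OntheKnees branch selects mount corrective offsets, driven by thigh angles, spherical pose readers and state flags, some interpolated.
# The DBNO offset is disconnected from execution.
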ControlRigBlueprint /BRCosmetics/Characters/Player/Male/Medium/Bodies/M_MED_FluffWoof/Meshes/Parts/M_MED_FluffWoof_FaceAcc_CtrlRig
graphs [
 New Function_ContainedGraph {nodes [
   n1 = RigVMFunctionEntryNode()
   RigVMFunction_MathBoolToInteger = &RigVMFunction_MathBoolToInteger(Value=@n1.onlyTwist)
   RigVMFunction_MathQuaternionSwingTwist = &RigVMFunction_MathQuaternionSwingTwist(
    TwistAxis=@n1.twistAxis
    Input=&RigUnit_GetRelativeTransformForItem(
     Child=(Type=Bone,Name="None")
     Parent=(Type=Bone,Name="None")
     bChildInitial=False
     bParentInitial=False
     Child.Name=@n1.child
     Parent.Name=@n1.parent
     Parent.Type=@n1.parentType
    ).RelativeTransform.Rotation
   )
   RigVMFunction_MathQuaternionToEuler = &RigVMFunction_MathQuaternionToEuler(
    Value=RigVMDispatchNode(
     Values=((),())
     Index=@RigVMFunction_MathBoolToInteger.Result
     Values.0=@RigVMFunction_MathQuaternionSwingTwist.Swing
     Values.1=@RigVMFunction_MathQuaternionSwingTwist.Twist
    ) {template="DISPATCH_RigVMDispatch_SelectInt32(in Index,in Values,out Result)"}.Result
    RotationOrder=@n1.Rotation Order
   )
   > DISPATCH_RigVMDispatch_Print = RigVMDispatchNode(ScreenColor=(R=1,G=1,B=1,A=1), ScreenDuration=1.5, Enabled=@n1.debugRotation, Value=@RigVMFunction_MathQuaternionToEuler.Result) {template="DISPATCH_RigVMDispatch_Print(in Prefix,in Value,in Enabled,in ScreenDuration,in ScreenColor)"}
   Clamp = &RigVMFunction_MathIntClamp(Value=@n1.primAxis)
   DISPATCH_RigVMDispatch_SelectInt32 = RigVMDispatchNode(
    Values=(0,0,0)
    Index=@Clamp.Result
    Values.0=@RigVMFunction_MathQuaternionToEuler.Result.X
    Values.1=@RigVMFunction_MathQuaternionToEuler.Result.Y
    Values.2=@RigVMFunction_MathQuaternionToEuler.Result.Z
   ) {template="DISPATCH_RigVMDispatch_SelectInt32(in Index,in Values,out Result)"}
   Remap = &RigVMFunction_MathDoubleRemap(
    TargetMaximum=1
    TargetMinimum=0
    bClamp=False
    Value=@DISPATCH_RigVMDispatch_SelectInt32.Result
    SourceMaximum=@n1.max
    SourceMinimum=@n1.min
   )
   Multiply = &RigVMFunction_MathDoubleMul(A=&RigVMFunction_MathDoubleClamp(Value=@Remap.Result).Result, B=@n1.Premult)
   > RigVMFunctionReturnNode(outputWeight=@Multiply.Result)
  ]}
 RigVMFunctionLibrary {nodes [DeltaRotation = RigVMCollapseNode(ExecuteContext=(), debugRotation=False, onlyTwist=False, twistAxis=(X=0,Y=0,Z=0))]}
 RigVMFunction_Sequence_2_ContainedGraph {nodes [
   n1 = RigVMFunctionEntryNode()
   n2 = &RigVMFunction_Sequence(ExecuteContext=@n1.ExecuteContext)
   n3 = &RigVMFunction_Sequence(ExecuteContext=@n2.B)
   n4 = &RigVMFunction_Sequence(ExecuteContext=@n3.B)
   n5 = &RigVMFunction_Sequence(ExecuteContext=@n4.B)
   RigVMFunctionReturnNode(A=@n2.A, B=@n3.A, C=@n4.A, D=@n5.A, E=@n5.B)
  ]}
 RigVMModel {nodes [
   RigUnit_BeginExecution = &RigUnit_BeginExecution()
   RigVMFunction_Sequence_1 = RigVMAggregateNode(ExecuteContext=@RigUnit_BeginExecution.ExecutePin)
   &RigUnit_ParentConstraint(
    AdvancedSettings=(InterpolationType=Average,RotationOrderForFilter=XZY)
    Child=(Type=Bone,Name="deform_clavicle_l")
    Filter=(TranslationFilter=(bX=False,bY=False,bZ=False),RotationFilter=(bX=True,bY=True,bZ=True),ScaleFilter=(bX=True,bY=True,bZ=True))
    Parents=((Item=(Type=Bone,Name="clavicle_l"),Weight=1))
    Weight=1
    bMaintainOffset=true
    ExecutePin=@RigVMFunction_Sequence_1.B
   )
   > &RigUnit_ParentConstraint(
    AdvancedSettings=(InterpolationType=Average,RotationOrderForFilter=XZY)
    Child=(Type=Bone,Name="deform_clavicle_r")
    Filter=(TranslationFilter=(bX=False,bY=False,bZ=False),RotationFilter=(bX=True,bY=True,bZ=True),ScaleFilter=(bX=True,bY=True,bZ=True))
    Parents=((Item=(Type=Bone,Name="clavicle_r"),Weight=1))
    Weight=1
    bMaintainOffset=true
   )
   n1 = &RigVMFunction_MathBoolToFloat(Value=RigVMVariableNode() {variable=NotRBAN}.Value)
   > &RigUnit_ParentConstraint(
    AdvancedSettings=(RotationOrderForFilter=XZY)
    Child=(Type=Bone,Name="aim_tie_02_proxy")
    Parents="((Item=(Type=Bone,Name=\"deform_pec_r\"),Weight=0.0024),(Item=(Type=Bone,Name=\"upperarm_twist_01_l\"),Weight=0.0028),(Item=(Type=Bone,Name=\"upperarm_twist_01_r\"),Weight=0.0028),(Item=(Type=Bone,Name=\"deform_clavicle_l\"),Weight=0.0042),(Item=(Type=Bone,Name=\"d ...(truncated)"
    Weight=@n1.Result
   )
   > &RigUnit_AimConstraintLocalSpaceOffset(
    Child=(Type=Bone,Name="dyn_tie_01")
    Filter=(bX=False,bY=True,bZ=True)
    Parents=((Item=(Type=Bone,Name="aim_tie_02_proxy"),Weight=1))
    WorldUp=(Target=(X=0,Y=0,Z=1),Kind=Direction,Space=(Type=None,Name="None"))
    Weight=@n1.Result
   )
   > &RigUnit_ParentConstraint(
    AdvancedSettings=(RotationOrderForFilter=XZY)
    Child=(Type=Bone,Name="aim_tie_03_proxy")
    Parents="((Item=(Type=Bone,Name=\"deform_clavicle_l\"),Weight=0.0052),(Item=(Type=Bone,Name=\"deform_clavicle_r\"),Weight=0.0052),(Item=(Type=Bone,Name=\"spine_01\"),Weight=0.0055),(Item=(Type=Bone,Name=\"pelvis\"),Weight=0.0073),(Item=(Type=Bone,Name=\"spine_02\"),Weight=0. ...(truncated)"
    Weight=@n1.Result
   )
   > &RigUnit_AimConstraintLocalSpaceOffset(
    Child=(Type=Bone,Name="dyn_tie_02")
    Filter=(bX=False,bY=True,bZ=True)
    Parents=((Item=(Type=Bone,Name="aim_tie_03_proxy"),Weight=1))
    WorldUp=(Target=(X=0,Y=0,Z=1),Kind=Direction,Space=(Type=None,Name="None"))
    Weight=@n1.Result
   )
   > &RigUnit_ParentConstraint(
    AdvancedSettings=(RotationOrderForFilter=XZY)
    Child=(Type=Bone,Name="aim_tie_04_proxy")
    Parents="((Item=(Type=Bone,Name=\"deform_clavicle_l\"),Weight=0.0015),(Item=(Type=Bone,Name=\"deform_clavicle_r\"),Weight=0.0015),(Item=(Type=Bone,Name=\"spine_05\"),Weight=0.0087),(Item=(Type=Bone,Name=\"spine_04\"),Weight=0.0591),(Item=(Type=Bone,Name=\"spine_01\"),Weight= ...(truncated)"
    Weight=@n1.Result
   )
   > DeltaRotation = RigVMFunctionReferenceNode(Premult=1, child=spine_05, debugRotation=False, max=20, onlyTwist=False, parent=pelvis, parentType=Bone, primAxis=2, twistAxis=(X=1,Y=0,Z=0))
   > &RigUnit_ModifyTransforms(
    ItemToModify=((Item=(Type=Bone,Name="aim_tie_04_proxy"),Transform=(Rotation=(X=0,Y=0,Z=0,W=1),Translation=(X=0,Y=1,Z=0),Scale3D=(X=1,Y=1,Z=1))))
    WeightMinimum=0
    Weight=&RigVMFunction_MathDoubleMul(A=@DeltaRotation.outputWeight, B=@n1.Result).Result
   )
   > &RigUnit_AimConstraintLocalSpaceOffset(
    Child=(Type=Bone,Name="dyn_tie_03")
    Filter=(bX=False,bY=True,bZ=True)
    Parents=((Item=(Type=Bone,Name="aim_tie_04_proxy"),Weight=1))
    WorldUp=(Target=(X=0,Y=0,Z=1),Kind=Direction,Space=(Type=None,Name="None"))
    Weight=@n1.Result
   )
   > &RigUnit_ParentConstraint(
    AdvancedSettings=(RotationOrderForFilter=XZY)
    Child=(Type=Bone,Name="aim_top_strap00_proxy")
    Parents=((Item=(Type=Bone,Name="thigh_twist_01_l"),Weight=0.0075),(Item=(Type=Bone,Name="spine_03"),Weight=0.0416),(Item=(Type=Bone,Name="spine_01"),Weight=0.0608),(Item=(Type=Bone,Name="spine_02"),Weight=0.0709),(Item=(Type=Bone,Name="pelvis"),Weight=0.8192))
    Weight=@n1.Result
   )
   > &RigUnit_AimConstraintLocalSpaceOffset(
    Child=(Type=Bone,Name="dyn_tie_04")
    Filter=(bX=False,bY=True,bZ=True)
    Parents=((Item=(Type=Bone,Name="aim_top_strap00_proxy"),Weight=1))
    WorldUp=(Target=(X=0,Y=0,Z=1),Kind=Direction,Space=(Type=None,Name="None"))
    Weight=@n1.Result
   )
   n2 = &RigUnit_ParentConstraint(
    AdvancedSettings=(RotationOrderForFilter=XZY)
    Child=(Type=Bone,Name="base_mount_proxy")
    Parents=((Item=(Type=Bone,Name="deform_glute_l"),Weight=0.0017),(Item=(Type=Bone,Name="spine_01"),Weight=0.0056),(Item=(Type=Bone,Name="thigh_twist_01_l"),Weight=0.4041),(Item=(Type=Bone,Name="pelvis"),Weight=0.5886))
    Weight=1
    bMaintainOffset=true
    ExecutePin=@RigVMFunction_Sequence_1.C
   )
   RigVMFunction_ControlFlowBranch = &RigVMFunction_ControlFlowBranch(ExecuteContext=@n2.ExecutePin, Condition=RigVMVariableNode() {variable=OntheKnees}.Value)
   &RigUnit_ModifyTransforms(
    ItemToModify="((Item=(Type=Bone,Name=\"base_mount_proxy\"),Transform=(Rotation=(X=-0.0059,Y=0.3199,Z=0.1917,W=0.9278),Translation=(X=4.9713,Y=-0.3535,Z=7.4959),Scale3D=(X=1,Y=1,Z=1))),(Item=(Type=Bone,Name=\"base_scissors00\"),Transform=(Rotation=(X=0.1736,Y=-0.0039,Z=-0.02 ...(truncated)"
    Weight=1
    WeightMinimum=0
    ExecutePin=@RigVMFunction_ControlFlowBranch.True
   )
   > &RigUnit_ParentConstraint(
    AdvancedSettings=(InterpolationType=Average,RotationOrderForFilter=XZY)
    Child=(Type=Bone,Name="base_mount")
    Parents=((Item=(Type=Bone,Name="base_mount_proxy"),Weight=1))
    Weight=1
    bMaintainOffset=true
   )
   n3 = &RigVMFunction_MathQuaternionToEuler(
    RotationOrder=ZYX
    Value=&RigUnit_GetRelativeTransformForItem(Child=(Type=Bone,Name="thigh_l"), Parent=(Type=Bone,Name="pelvis"), bChildInitial=False, bParentInitial=False).RelativeTransform.Rotation
   )
   > &RigUnit_ModifyTransforms(
    ItemToModify="((Item=(Type=Bone,Name=\"base_mount\"),Transform=(Rotation=(X=-0.0538,Y=-0.1675,Z=-0.1223,W=0.9768),Translation=(X=-0.7589,Y=0.6567,Z=-5.4319),Scale3D=(X=1,Y=1,Z=1))),(Item=(Type=Bone,Name=\"base_scissors00\"),Transform=(Rotation=(X=0.0868,Y=0.0076,Z=0.0868,W= ...(truncated)"
    WeightMinimum=0
    Weight=RigVMDispatchNode(
     True=0
     False=&RigVMFunction_MathDoubleRemap(SourceMaximum=-18, SourceMinimum=-79, TargetMaximum=1, TargetMinimum=0, Value=@n3.Result.Z).Result
     Condition=RigVMVariableNode() {variable=KneeSlide}.Value
    ) {template="DISPATCH_RigVMDispatch_If(in Condition,in True,in False,out Result)"}.Result
   )
   > n4 = &RigUnit_SphericalPoseReader(
    ActiveRegionScaleFactors=(PositiveWidth=0.7,NegativeWidth=1,PositiveHeight=1,NegativeHeight=0.7)
    ActiveRegionSize=0.06
    Debug=(bDrawDebug=true,bDraw2D=False,bDrawLocalAxes=False,DebugScale=25,DebugSegments=20,DebugThickness=0.25)
    DriverAxis=(X=-1,Y=0,Z=0)
    DriverItem=(Type=Bone,Name="thigh_l")
    FalloffRegionScaleFactors=(PositiveWidth=0.8,NegativeWidth=2,PositiveHeight=2,NegativeHeight=1)
    FalloffSize=0.2
    FlipHeightScaling=False
    FlipWidthScaling=False
    OptionalParentItem=(Type=Bone,Name="None")
    RotationOffset=(X=130,Y=-30,Z=0)
   )
   n5 = &RigVMFunction_AlphaInterp(
    InterpSpeedDecreasing=5
    InterpSpeedIncreasing=5
    Scale=1
    bClampResult=False
    Value=RigVMDispatchNode(False=0, True=@n4.OutputParam, Condition=RigVMVariableNode() {variable=Movin2D}.Value) {template="DISPATCH_RigVMDispatch_If(in Condition,in True,in False,out Result)"}.Result
   )
   > &RigUnit_ModifyTransforms(
    ItemToModify="((Item=(Type=Bone,Name=\"base_mount\"),Transform=(Rotation=(X=0,Y=0,Z=0.0436,W=0.999),Translation=(X=-2.4664,Y=1.5,Z=1.3854),Scale3D=(X=1,Y=1,Z=1))),(Item=(Type=Bone,Name=\"base_scissors00\"),Transform=(Rotation=(X=0.1736,Y=-0.003,Z=-0.0172,W=0.9847),Translati ...(truncated)"
    WeightMinimum=0
    Weight=@n5.Result
   )
   n6 = &RigVMFunction_AlphaInterp(
    InterpSpeedDecreasing=5
    InterpSpeedIncreasing=5
    Scale=1
    bClampResult=False
    Value=RigVMDispatchNode(False=0, True=@n4.OutputParam, Condition=RigVMVariableNode() {variable=Aiming}.Value) {template="DISPATCH_RigVMDispatch_If(in Condition,in True,in False,out Result)"}.Result
   )
   > &RigUnit_ModifyTransforms(
    ItemToModify="((Item=(Type=Bone,Name=\"base_mount\"),Transform=(Rotation=(X=-0.1006,Y=-0.1648,Z=0.0704,W=0.9786),Translation=(X=0.9703,Y=1.0687,Z=-1.3755),Scale3D=(X=1,Y=1,Z=1))),(Item=(Type=Bone,Name=\"base_scissors01\"),Transform=(Rotation=(X=-0.0076,Y=0.0868,Z=0.0868,W=0 ...(truncated)"
    WeightMinimum=0
    Weight=@n6.Result
   )
   RigVMFunction_MathQuaternionToEuler = &RigVMFunction_MathQuaternionToEuler(
    RotationOrder=ZYX
    Value=&RigUnit_GetRelativeTransformForItem(Child=(Type=Bone,Name="thigh_l"), Parent=(Type=Bone,Name="pelvis"), bChildInitial=False, bParentInitial=False).RelativeTransform.Rotation
   )
   Remap = &RigVMFunction_MathDoubleRemap(SourceMaximum=6, SourceMinimum=30, TargetMaximum=0, TargetMinimum=1, Value=@RigVMFunction_MathQuaternionToEuler.Result.Y)
   > &RigUnit_ModifyTransforms(
    ItemToModify=((Item=(Type=Bone,Name="base_mount"),Transform=(Rotation=(X=0.1736,Y=0,Z=0,W=0.9848),Translation=(X=-1.4106,Y=-1.2203,Z=-0.2152),Scale3D=(X=1,Y=1,Z=1))))
    WeightMinimum=0
    Weight=@Remap.Result
   )
   AlphaInterp = &RigVMFunction_AlphaInterp(
    InterpSpeedDecreasing=10
    InterpSpeedIncreasing=10
    Scale=1
    bClampResult=False
    Value=RigVMDispatchNode(False=0, True=1, Condition=RigVMVariableNode() {variable=KneeSlide}.Value) {template="DISPATCH_RigVMDispatch_If(in Condition,in True,in False,out Result)"}.Result
   )
   > &RigUnit_ModifyTransforms(
    ItemToModify="((Item=(Type=Bone,Name=\"base_mount\"),Transform=(Rotation=(X=0.0258,Y=-0.1764,Z=-0.1366,W=0.9744),Translation=(X=-0.7171,Y=0.7,Z=-2.8803),Scale3D=(X=1,Y=1,Z=1))),(Item=(Type=Bone,Name=\"base_bot_strap00\"),Transform=(Rotation=(X=0,Y=0.2588,Z=0,W=0.9659),Trans ...(truncated)"
    WeightMinimum=0
    Weight=@AlphaInterp.Result
   )
   n7 = &RigUnit_SphericalPoseReader(
    ActiveRegionScaleFactors=(PositiveWidth=1,NegativeWidth=1,PositiveHeight=1,NegativeHeight=1)
    ActiveRegionSize=0.05
    Debug=(bDrawDebug=false,bDraw2D=false,bDrawLocalAxes=False,DebugScale=25,DebugSegments=20,DebugThickness=0.25)
    DriverAxis=(X=-1,Y=0,Z=0)
    DriverItem=(Type=Bone,Name="thigh_l")
    FalloffRegionScaleFactors=(PositiveWidth=1,NegativeWidth=1,PositiveHeight=1,NegativeHeight=1)
    FalloffSize=0.35
    FlipHeightScaling=False
    FlipWidthScaling=False
    OptionalParentItem=(Type=Bone,Name="None")
    RotationOffset=(X=90,Y=0,Z=0)
    ExecutePin=@RigVMFunction_ControlFlowBranch.False
   )
   > &RigUnit_ModifyTransforms(
    ItemToModify=((Item=(Type=Bone,Name="base_mount"),Transform=(Rotation=(X=0.0076,Y=0.0868,Z=0.0868,W=0.9924),Translation=(X=0,Y=0,Z=0),Scale3D=(X=1,Y=1,Z=1))))
    WeightMinimum=0
    Weight=RigVMDispatchNode(False=0, True=@n7.OutputParam, Condition=RigVMVariableNode() {variable=Emoting}.Value) {template="DISPATCH_RigVMDispatch_If(in Condition,in True,in False,out Result)"}.Result
   )
   > &RigUnit_ModifyTransforms(
    ItemToModify="((Item=(Type=Bone,Name=\"base_mount_proxy\"),Transform=(Rotation=(X=0,Y=0,Z=0.0872,W=0.9962),Translation=(X=0.25,Y=-0.4001,Z=3),Scale3D=(X=1,Y=1,Z=1))),(Item=(Type=Bone,Name=\"base_mount01\"),Transform=(Rotation=(X=0,Y=0,Z=-0.1736,W=0.9848),Translation=(X=0,Y= ...(truncated)"
    Weight=1
    WeightMinimum=0
   )
   > &RigUnit_ParentConstraint(
    AdvancedSettings=(InterpolationType=Average,RotationOrderForFilter=XZY)
    Child=(Type=Bone,Name="base_mount")
    Parents=((Item=(Type=Bone,Name="base_mount_proxy"),Weight=1))
    Weight=1
    bMaintainOffset=true
   )
   > n8 = &RigUnit_SphericalPoseReader(
    ActiveRegionScaleFactors=(PositiveWidth=1,NegativeWidth=1,PositiveHeight=1,NegativeHeight=1)
    ActiveRegionSize=0.05
    Debug=(bDrawDebug=false,bDraw2D=false,bDrawLocalAxes=False,DebugScale=25,DebugSegments=20,DebugThickness=0.25)
    DriverAxis=(X=-1,Y=0,Z=0)
    DriverItem=(Type=Bone,Name="thigh_l")
    FalloffRegionScaleFactors=(PositiveWidth=1,NegativeWidth=1,PositiveHeight=1,NegativeHeight=1)
    FalloffSize=0.3
    FlipHeightScaling=False
    FlipWidthScaling=False
    OptionalParentItem=(Type=Bone,Name="None")
    RotationOffset=(X=90,Y=0,Z=0)
   )
   n9 = RigVMDispatchNode(True=0, False=@n8.OutputParam, Condition=RigVMVariableNode() {variable=jumping}.Value) {template="DISPATCH_RigVMDispatch_If(in Condition,in True,in False,out Result)"}
   > &RigUnit_ModifyTransforms(
    ItemToModify="((Item=(Type=Bone,Name=\"base_mount\"),Transform=(Rotation=(X=0,Y=0,Z=0.0523,W=0.9986),Translation=(X=0,Y=0,Z=0),Scale3D=(X=1,Y=1,Z=1))),(Item=(Type=Bone,Name=\"base_scissors00\"),Transform=(Rotation=(X=0.1736,Y=0,Z=0,W=0.9848),Translation=(X=0,Y=0,Z=0),Scale3 ...(truncated)"
    WeightMinimum=0
    Weight=@n9.Result
   )
   > &RigUnit_ModifyTransforms(
    ItemToModify="((Item=(Type=Bone,Name=\"base_top_strap00\"),Transform=(Rotation=(X=0,Y=-0.5,Z=0,W=0.866),Translation=(X=0,Y=0,Z=0),Scale3D=(X=1,Y=1,Z=1))),(Item=(Type=Bone,Name=\"base_bot_strap00\"),Transform=(Rotation=(X=0,Y=0.2588,Z=0,W=0.9659),Translation=(X=0,Y=0,Z=0),Sc ...(truncated)"
    WeightMinimum=0
    Weight=@n9.Result
   )
   > n10 = &RigUnit_SphericalPoseReader(
    ActiveRegionScaleFactors=(PositiveWidth=1,NegativeWidth=1.2,PositiveHeight=1,NegativeHeight=1)
    ActiveRegionSize=0.12
    Debug=(bDrawDebug=false,bDraw2D=False,bDrawLocalAxes=False,DebugScale=25,DebugSegments=20,DebugThickness=0.25)
    DriverAxis=(X=-1,Y=0,Z=0)
    DriverItem=(Type=Bone,Name="thigh_l")
    FalloffRegionScaleFactors=(PositiveWidth=1,NegativeWidth=1,PositiveHeight=1,NegativeHeight=1)
    FalloffSize=0.3
    FlipHeightScaling=False
    FlipWidthScaling=False
    OptionalParentItem=(Type=Bone,Name="None")
    RotationOffset=(X=180,Y=-30,Z=0)
   )
   > &RigUnit_ModifyTransforms(
    ItemToModify="((Item=(Type=Bone,Name=\"base_scissors00\"),Transform=(Rotation=(X=0,Y=0,Z=-0.0349,W=0.9994),Translation=(X=0.0532,Y=-2,Z=1),Scale3D=(X=1,Y=1,Z=1))),(Item=(Type=Bone,Name=\"base_mount\"),Transform=(Rotation=(X=0,Y=0,Z=0.0872,W=0.9962),Translation=(X=0,Y=1,Z=0) ...(truncated)"
    WeightMinimum=0
    Weight=RigVMDispatchNode(True=0, False=@n10.OutputParam, Condition=RigVMVariableNode() {variable=jumping}.Value) {template="DISPATCH_RigVMDispatch_If(in Condition,in True,in False,out Result)"}.Result
   )
   > n11 = &RigUnit_SphericalPoseReader(
    ActiveRegionScaleFactors=(PositiveWidth=1,NegativeWidth=1,PositiveHeight=1,NegativeHeight=1)
    ActiveRegionSize=0.07
    Debug=(bDrawDebug=false,bDraw2D=False,bDrawLocalAxes=False,DebugScale=25,DebugSegments=20,DebugThickness=0.25)
    DriverAxis=(X=-1,Y=0,Z=0)
    DriverItem=(Type=Bone,Name="thigh_l")
    FalloffRegionScaleFactors=(PositiveWidth=1,NegativeWidth=1,PositiveHeight=1,NegativeHeight=1)
    FalloffSize=0.23
    FlipHeightScaling=False
    FlipWidthScaling=False
    OptionalParentItem=(Type=Bone,Name="None")
    RotationOffset=(X=35,Y=45,Z=0)
   )
   > &RigUnit_ModifyTransforms(
    ItemToModify="((Item=(Type=Bone,Name=\"base_mount\"),Transform=(Rotation=(X=0,Y=-0.0872,Z=0,W=0.9962),Translation=(X=0,Y=1,Z=1),Scale3D=(X=1,Y=1,Z=1))),(Item=(Type=Bone,Name=\"base_scissors00\"),Transform=(Rotation=(X=0,Y=0,Z=0,W=1),Translation=(X=0,Y=-3,Z=0),Scale3D=(X=1,Y ...(truncated)"
    WeightMinimum=0
    Weight=RigVMDispatchNode(False=0, True=@n11.OutputParam, Condition=RigVMVariableNode() {variable=jumping}.Value) {template="DISPATCH_RigVMDispatch_If(in Condition,in True,in False,out Result)"}.Result
   )
   RigVMFunction_MathBoolNot = &RigVMFunction_MathBoolNot(Value=RigVMVariableNode() {variable=NotFrontEnd}.Value)
   RigVMFunction_MathBoolAnd = &RigVMFunction_MathBoolAnd(A=@RigVMFunction_MathBoolNot.Result, B=&RigVMFunction_MathBoolNot(Value=RigVMVariableNode() {variable=Emoting}.Value).Result)
   > &RigUnit_ModifyTransforms(
    ItemToModify=((Item=(Type=Bone,Name="base_mount"),Transform=(Rotation=(X=0,Y=0,Z=0,W=1),Translation=(X=-1.5,Y=0,Z=0),Scale3D=(X=1,Y=1,Z=1))))
    WeightMinimum=0
    Weight=&RigVMFunction_MathBoolToFloat(Value=@RigVMFunction_MathBoolAnd.Result).Result
   )
   > n12 = &RigUnit_SphericalPoseReader(
    ActiveRegionScaleFactors=(PositiveWidth=1,NegativeWidth=1.2,PositiveHeight=1,NegativeHeight=1)
    ActiveRegionSize=0.12
    Debug=(bDrawDebug=false,bDraw2D=False,bDrawLocalAxes=False,DebugScale=25,DebugSegments=20,DebugThickness=0.25)
    DriverAxis=(X=-1,Y=0,Z=0)
    DriverItem=(Type=Bone,Name="thigh_l")
    FalloffRegionScaleFactors=(PositiveWidth=1,NegativeWidth=1,PositiveHeight=1,NegativeHeight=1)
    FalloffSize=0.3
    FlipHeightScaling=False
    FlipWidthScaling=False
    OptionalParentItem=(Type=Bone,Name="None")
    RotationOffset=(X=-90,Y=0,Z=-45)
   )
   n13 = RigVMVariableNode() {variable=Emoting}
   > &RigUnit_ModifyTransforms(
    ItemToModify=((Item=(Type=Bone,Name="base_mount"),Transform=(Rotation=(X=0,Y=0,Z=0,W=1),Translation=(X=-1.5,Y=0,Z=0),Scale3D=(X=1,Y=1,Z=1))))
    WeightMinimum=0
    Weight=&RigVMFunction_MathDoubleMul(A=&RigVMFunction_MathBoolToFloat(Value=@n13.Value).Result, B=@n12.OutputParam).Result
   )
   RigVMFunction_MathBoolToFloat = &RigVMFunction_MathBoolToFloat(Value=RigVMVariableNode() {variable=DBNO}.Value)
   n14 = &RigVMFunction_AlphaInterp(InterpSpeedDecreasing=5, InterpSpeedIncreasing=5, Scale=1, bClampResult=False, Value=@RigVMFunction_MathBoolToFloat.Result)
   Multiply = &RigVMFunction_MathDoubleMul(B=@n14.Result, A=@n4.OutputParam)
   &RigUnit_ModifyTransforms(
    ExecutePin=()
    ItemToModify="((Item=(Type=Bone,Name=\"base_mount\"),Transform=(Rotation=(X=-0.1006,Y=-0.1648,Z=0.0704,W=0.9786),Translation=(X=0.9703,Y=1.0687,Z=-1.3755),Scale3D=(X=1,Y=1,Z=1))),(Item=(Type=Bone,Name=\"base_scissors01\"),Transform=(Rotation=(X=-0.0076,Y=0.0868,Z=0.0868,W=0 ...(truncated)"
    WeightMinimum=0
    Weight=@Multiply.Result
   )
  ]}
]
hierarchy [
 root {loc=<0 0 0>, rot=<0 0 0 1>}
 pelvis {parent=root, loc=<0 -0.7071 93.1191>, rot=<0.0127 0.707 -0.0127 -0.707>}
 spine_01 {parent=pelvis, loc=<6.7494 -0.6388 0>, rot=<0 0 0.034 -0.9994>}
 spine_02 {parent=spine_01, loc=<7.252 -0.035 0>, rot=<0 0 0.0238 -0.9997>}
 spine_03 {parent=spine_02, loc=<6.0106 0 0>, rot=<0 0 0.0149 -0.9999>}
 spine_04 {parent=spine_03, loc=<8.1023 0 0>, rot=<0 0 -0.0132 -0.9999>}
 spine_05 {parent=spine_04, loc=<10.0142 0 0>, rot=<0 0 -0.0015 -1>}
 neck_01 {parent=spine_05, loc=<11.3138 0.1529 0>, rot=<0 0 -0.2003 -0.9797>}
 neck_02 {parent=neck_01, loc=<4.5627 0 0>, rot=<0 0 0.0504 -0.9987>}
 head {parent=neck_02, loc=<5.3914 0 0>, rot=<0 0 0.115 -0.9934>}
 base_tie_00 {parent=spine_05, loc=<11.3736 7.5924 0>, rot=<-0.1642 -0.9864 0 0>}
 dyn_tie_00 {parent=base_tie_00, loc=<0 0 0>, rot=<0 0 -0.0189 -0.9998>}
 base_tie_01 {parent=dyn_tie_00, loc=<2.5667 -0.097 0>, rot=<0 0 0 1>}
 dyn_tie_01 {parent=base_tie_01, loc=<0 0 0>, rot=<0 0 0 1>}
 dyn_tie_02 {parent=dyn_tie_01, loc=<7.398 0 0>, rot=<0 0 0.0811 -0.9967>}
 dyn_tie_03 {parent=dyn_tie_02, loc=<10.7289 0 0>, rot=<0 0 0.1478 -0.989>}
 dyn_tie_04 {parent=dyn_tie_03, loc=<8.7657 0.0718 0>, rot=<0 0 0.0143 -0.9999>}
 clavicle_l {parent=spine_05, loc=<4.3575 -0.1466 -2.3223>, rot=<0.0465 -0.7289 -0.009 -0.6829>}
 upperarm_l {parent=clavicle_l, loc=<15.5111 0 0>, rot=<0.0226 -0.3234 -0.0088 -0.9459>}
 lowerarm_l {parent=upperarm_l, loc=<26.6514 0 0>, rot=<0.0391 -0.0099 -0.2453 -0.9686>}
 upperarm_twist_01_l {parent=upperarm_l, loc=<0.0041 0 0>, rot=<0 0 0 -1>}
 deform_upperarm_bk_base_l {parent=upperarm_twist_01_l, loc=<1.0813 -7.2193 -1.9453>, rot=<0 0 0 1>}
 deform_upperarm_bk_l {parent=deform_upperarm_bk_base_l, loc=<0 0 0>, rot=<0 0 0 1>}
 deform_upperarm_fr_base_l {parent=upperarm_twist_01_l, loc=<-0.3691 7.0506 -2.765>, rot=<0 0 0 1>}
 deform_upperarm_fr_l {parent=deform_upperarm_fr_base_l, loc=<0 0 0>, rot=<0 0 0 1>}
 deform_upperarm_up_base_l {parent=upperarm_twist_01_l, loc=<-3.2216 -1.1391 3.3311>, rot=<0 0 0 1>}
 deform_upperarm_up_l {parent=deform_upperarm_up_base_l, loc=<0 0 0>, rot=<0 0 0 1>}
 clavicle_r {parent=spine_05, loc=<4.3575 -0.1466 2.3223>, rot=<0.7289 0.0465 -0.6829 0.009>}
 upperarm_r {parent=clavicle_r, loc=<-15.5111 0 0>, rot=<0.0226 -0.3234 -0.0088 -0.9459>}
 lowerarm_r {parent=upperarm_r, loc=<-26.6514 0 0>, rot=<0.0391 -0.0099 -0.2453 -0.9686>}
 upperarm_twist_01_r {parent=upperarm_r, loc=<-0.0041 0 0>, rot=<0 0 0 -1>}
 deform_upperarm_bk_base_r {parent=upperarm_twist_01_r, loc=<-1.0814 7.2192 1.9455>, rot=<0 0 0 1>}
 deform_upperarm_bk_r {parent=deform_upperarm_bk_base_r, loc=<0 0 0>, rot=<0 0 0 1>}
 deform_upperarm_fr_base_r {parent=upperarm_twist_01_r, loc=<0.3689 -7.0506 2.7652>, rot=<0 0 0 1>}
 deform_upperarm_fr_r {parent=deform_upperarm_fr_base_r, loc=<0 0 0>, rot=<0 0 0 1>}
 deform_upperarm_up_base_r {parent=upperarm_twist_01_r, loc=<3.2214 1.1391 -3.3308>, rot=<0 0 0 1>}
 deform_upperarm_up_r {parent=deform_upperarm_up_base_r, loc=<0 0 0>, rot=<0 0 0 -1>}
 deform_clavicle_l {parent=spine_05, loc=<9.5308 0.0818 -6.0277>, rot=<0.0465 -0.7289 -0.009 -0.6829>}
 deform_clavicle_r {parent=spine_05, loc=<9.5311 0.0819 6.0277>, rot=<0.7289 0.0465 -0.6829 0.009>}
 deform_neck_01 {parent=spine_05, loc=<15.5491 1.9486 0>, rot=<0 0 -0.1329 -0.9911>}
 deform_neck_02 {parent=deform_neck_01, loc=<4.5627 0 0>, rot=<0 0 -0.0112 -0.9999>}
 deform_head {parent=deform_neck_02, loc=<5.3914 0 0>, rot=<0 0 0.0595 -0.9982>}
 deform_pec_aim_base_l {parent=spine_04, loc=<7.8728 12.6388 -2.4123>, rot=<0.0102 -0.4602 0.3979 -0.7936>}
 deform_pec_aim_l {parent=deform_pec_aim_base_l, loc=<0 0 0>, rot=<0 0 0 1>}
 deform_pec_base_l {parent=deform_pec_aim_l, loc=<10.766 0 0>, rot=<0 0 0 1>}
 deform_pec_l {parent=deform_pec_base_l, loc=<0.055 0 0>, rot=<0 0 0 1>}
 deform_pec_aim_base_r {parent=spine_04, loc=<7.8723 12.6388 2.4123>, rot=<-0.4602 -0.0102 0.7936 0.3979>}
 deform_pec_aim_r {parent=deform_pec_aim_base_r, loc=<0 0 0>, rot=<0 0 0 1>}
 deform_pec_base_r {parent=deform_pec_aim_r, loc=<-10.7662 -0.0006 -0.0007>, rot=<0 0 0 1>}
 deform_pec_r {parent=deform_pec_base_r, loc=<0 0 0>, rot=<0 0 0 1>}
 thigh_l {parent=pelvis, loc=<-2.5693 -0.0109 -7.7501>, rot=<0.0012 0.0572 0.0155 -0.9982>}
 base_mount_proxy {parent=thigh_l, loc=<-3.3434 -6.5728 -5.8179>, rot=<-0.5344 -0.4979 0.5262 0.4354>}
 base_mount00_proxy {parent=base_mount_proxy, loc=<8.7145 -0.6985 -0.105>, rot=<-0.073 0.0222 -0.2532 0.9644>}
 base_mount01_proxy {parent=base_mount00_proxy, loc=<4.5212 -1.0001 0.3245>, rot=<-0.1907 -0.3037 0.3488 -0.8659>}
 thigh_twist_01_l {parent=thigh_l, loc=<-21.6303 0 0>, rot=<0.0013 0 0 -1>}
 base_mount {parent=thigh_l, loc=<-1.6851 -5.5146 -6.1436>, rot=<0.5344 0.4979 -0.5262 -0.4354>}
 dyn_mount {parent=base_mount, loc=<0 0 0>, rot=<0 0 0 1>}
 base_mount00 {parent=dyn_mount, loc=<8.7145 -0.6985 -0.105>, rot=<0.073 -0.0222 0.2532 -0.9644>}
 dyn_mount00 {parent=base_mount00, loc=<0 0 0>, rot=<0 0 0 1>}
 base_scissors00 {parent=base_mount00, loc=<3.2731 0.3476 0.3555>, rot=<0.7283 -0.1459 0.6661 0.0679>}
 dyn_scissors00 {parent=base_scissors00, loc=<0 0 0>, rot=<0 0 0 1>}
 base_scissors01 {parent=dyn_scissors00, loc=<0 0 0>, rot=<0 0 0 1>}
 base_mount01 {parent=base_mount00, loc=<4.5212 -1.0001 0.3245>, rot=<-0.1907 -0.3037 0.3488 -0.8659>}
 base_bot_strap00 {parent=base_mount01, loc=<0.6069 1.0392 0.1849>, rot=<0 0 0 1>}
 dyn_bot_strap00 {parent=base_bot_strap00, loc=<0 0 0>, rot=<0 0 0 1>}
 base_comb00 {parent=dyn_mount, loc=<7.2165 1.0829 0.4615>, rot=<0.7009 -0.0516 0.7112 -0.015>}
 dyn_comb00 {parent=base_comb00, loc=<0 0 0>, rot=<0 0 0 1>}
 thigh_r {parent=pelvis, loc=<-2.5693 -0.0109 7.7501>, rot=<0.0572 -0.0012 0.9982 0.0155>}
 thigh_twist_01_r {parent=thigh_r, loc=<21.6303 0 0>, rot=<0.0013 0 0 -1>}
 base_top_strap00 {parent=pelvis, loc=<4.3851 10.9596 -1.9175>, rot=<0.0295 -0.9407 0.1475 -0.3042>}
 dyn_top_strap00 {parent=base_top_strap00, loc=<0 0 0>, rot=<0 0 0 1>}
 base_mid_strap00 {parent=dyn_top_strap00, loc=<6.7621 -0.1756 -0.2958>, rot=<0.001 0.1204 -0.0086 -0.9927>}
 dyn_mid_strap00 {parent=base_mid_strap00, loc=<0 0 0>, rot=<0 0 0 1>}
 deform_glute_base_l {parent=pelvis, loc=<-8.8064 -8.8222 -8.4287>, rot=<0.0275 0.0501 -0.4472 -0.8926>}
 deform_glute_l {parent=deform_glute_base_l, loc=<0 0 0>, rot=<0 0 0 1>}
 deform_groin_base_l {parent=pelvis, loc=<-4.2941 8.1593 -9.7386>, rot=<0.0012 0.0572 0.0155 -0.9982>}
 deform_groin_l {parent=deform_groin_base_l, loc=<0 0 0>, rot=<0 0 0 1>}
 thigh_l_ctrl {controlType=EULER_TRANSFORM, type=Control, loc=<0 0 0>, rot=<0 0 0 1>}
 aim_tie_02_proxy {loc=<0 7.3147 134.0804>, rot=<-0.1006 -0.6999 -0.1006 -0.6999>}
 aim_tie_03_proxy {loc=<0 10.3367 123.7859>, rot=<-0.004 0.7071 -0.004 0.7071>}
 aim_tie_04_proxy {loc=<0 10.3105 115.0199>, rot=<0.014 -0.707 0.014 -0.707>}
 aim_tie_01_proxy {loc=<0 4.1112 140.7489>, rot=<0.157 0.6895 0.157 0.6895>}
 aim_top_strap00_proxy {loc=<1.9175 10.4025 97.1088>, rot=<0.0676 0.4477 -0.1332 0.8816>}
]
variables [
 Aiming {type=bool, default=False}
 CrouchAlpha {type=double, default=0}
 DBNO {type=bool, default=False}
 Emoting {type=bool, default=False}
 KneeSlide {type=bool, default=False}
 Movin2D {type=bool, default=False}
 NotFrontEnd {type=bool, default=False}
 NotRBAN {type=bool, default=False}
 OntheKnees {type=bool, default=False}
 jumping {type=bool, default=False}
]
