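# Physics asset for the front tie end on the FluffWoof medium male character body, set up for a dangling tie over the torso.
# Capsules cover the spine, neck and tie bones, with large torso capsules providing volume around the tie.
# Locked constraints chain the spine, neck and two base tie bones into a rigid anchor.
# Four dyn_tie bones hang in series from base_tie_01 on unlocked joints that share the same angle settings.
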PhysicsAsset /BRCosmetics/Characters/Player/Male/Medium/Bodies/M_MED_FluffWoof/Meshes/Parts/M_MED_FluffWoof_TieFrontEnd_Physics
bodies {
 base_tie_00 [{shapeType=Capsule, length=0.505, radius=0.505, shapeName=base_tie_00_capsule, rotation=<rot 0 -90 0>}]
 base_tie_01 [{shapeType=Capsule, length=0.505, radius=0.505, shapeName=base_tie_01_capsule, rotation=<rot 0 -90 0>}]
 dyn_tie_01 [{shapeType=Capsule, length=1.6077, radius=1.6394, shapeName=dyn_tie_02_capsule, center=<1.9214 1.4 0>}]
 dyn_tie_02 [{shapeType=Capsule, length=4.3359, radius=1.7675, shapeName=dyn_tie_03_capsule, center=<3.4725 0.7382 0>, rotation=<rot 90 0 0>}]
 dyn_tie_03 [{shapeType=Capsule, length=5.505, radius=2.1, shapeName=dyn_tie_04_capsule, center=<3.0418 1.4013 0>, rotation=<rot 90 180 -176.0446>}]
 dyn_tie_04 [{shapeType=Capsule, length=6.0401, radius=2.288, shapeName=dyn_tie_05_capsule, center=<4.3567 1.3357 0>, rotation=<rot 90 180 -177.6076>}]
 neck_01 [{shapeType=Capsule, length=0.5, radius=5, shapeName=neck_01_capsule, rotation=<rot 0 -90 0>}]
 neck_02 [{
   shapeType=Capsule
   length=7.2601
   radius=2.5276
   shapeName=neck_02_capsule
   center=<1.0002 -4.8906 -0.0001>
   rotation=<rot -0.0021 -90 -0.0057>
  }]
 spine_02 [{shapeType=Capsule, length=0.5, radius=13.005, shapeName=spine_02_capsule, center=<2.8545 -2.083 0>, rotation=<rot 0 -90 0>}]
 spine_03 [{shapeType=Capsule, length=10.005, radius=13.005, shapeName=spine_03_capsule, center=<4.8561 -1.8837 0>, rotation=<rot 0 -90 0>}]
 spine_04 [{shapeType=Capsule, length=15.505, radius=10.505, shapeName=spine_04_capsule, center=<1.1313 1.3937 0>, rotation=<rot 0 -90 0>}]
 spine_05 [{shapeType=Capsule, length=14.005, radius=9.505, shapeName=spine_05_capsule, center=<-0.7993 1.0071 0>, rotation=<rot 0 -90 0>}]
}
constraints [
 {
  bone1Name=spine_03
  bone2Name=spine_02
  swing1LimitDegrees=45
  swing1Motion=Locked
  swing2LimitDegrees=45
  swing2Motion=Locked
  twistLimitDegrees=45
  twistMotion=Locked
 }
 {
  bone1Name=spine_04
  bone2Name=spine_03
  swing1LimitDegrees=45
  swing1Motion=Locked
  swing2LimitDegrees=45
  swing2Motion=Locked
  twistLimitDegrees=45
  twistMotion=Locked
 }
 {
  bone1Name=spine_05
  bone2Name=spine_04
  swing1LimitDegrees=45
  swing1Motion=Locked
  swing2LimitDegrees=45
  swing2Motion=Locked
  twistLimitDegrees=45
  twistMotion=Locked
 }
 {
  bone1Name=neck_01
  bone2Name=spine_05
  swing1LimitDegrees=45
  swing1Motion=Locked
  swing2LimitDegrees=45
  swing2Motion=Locked
  twistLimitDegrees=45
  twistMotion=Locked
 }
 {bone1Name=dyn_tie_01, bone2Name=base_tie_01, swing1LimitDegrees=40, swing2LimitDegrees=15, twistLimitDegrees=0}
 {bone1Name=dyn_tie_02, bone2Name=dyn_tie_01, swing1LimitDegrees=40, swing2LimitDegrees=15, twistLimitDegrees=0}
 {bone1Name=dyn_tie_03, bone2Name=dyn_tie_02, swing1LimitDegrees=40, swing2LimitDegrees=15, twistLimitDegrees=0}
 {bone1Name=dyn_tie_04, bone2Name=dyn_tie_03, swing1LimitDegrees=40, swing2LimitDegrees=15, twistLimitDegrees=0}
 {
  bone1Name=neck_02
  bone2Name=neck_01
  swing1LimitDegrees=45
  swing1Motion=Locked
  swing2LimitDegrees=45
  swing2Motion=Locked
  twistLimitDegrees=45
  twistMotion=Locked
 }
 {
  bone1Name=base_tie_00
  bone2Name=spine_05
  swing1LimitDegrees=45
  swing1Motion=Locked
  swing2LimitDegrees=45
  swing2Motion=Locked
  twistLimitDegrees=45
  twistMotion=Locked
 }
 {
  bone1Name=base_tie_01
  bone2Name=base_tie_00
  swing1LimitDegrees=45
  swing1Motion=Locked
  swing2LimitDegrees=45
  swing2Motion=Locked
  twistLimitDegrees=45
  twistMotion=Locked
 }
]
