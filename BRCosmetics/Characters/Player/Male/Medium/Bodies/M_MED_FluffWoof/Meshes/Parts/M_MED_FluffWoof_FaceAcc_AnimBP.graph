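# Animation Blueprint for the FluffWoof face accessory part, adding procedural motion to its tie, scissors, comb and straps.
# It blends the attached parent's pose with body-mesh thigh and clavicle branches, then feeds a Control Rig driven by movement-state flags.
# Two RigidBody simulations and spring AnimDynamics for scissors and comb follow.
# Jump-gated additive bone offsets finish the pose, while the EventGraph updates state flags and a deprecated-function gravity override.
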AnimBlueprint /BRCosmetics/Characters/Player/Male/Medium/Bodies/M_MED_FluffWoof/Meshes/Parts/M_MED_FluffWoof_FaceAcc_AnimBP
target_skeleton=&/M_MED_FluffWoof_FaceAcc_Skeleton
graphs [
 AnimGraph {nodes [
   n1 = Control Rig!(
    Source=Use cached pose 'Start_Cache'()
    Alpha
    OntheKnees:bool=Get ontheKnees() -> (ontheKnees:bool)
    Movin2D:bool=Get bIsMoving2D()
    jumping:bool=Get Jumping() -> (Jumping:bool)
    Emoting:bool=Get bIsPlayingEmote()
    Aiming:bool=Get bIsTargeting()
    KneeSlide:bool=Get bIsKneeSliding()
    NotFrontEnd:bool=Get isNotFrontEnd() -> (isNotFrontEnd:bool)
    NotRBAN:bool=NOT Boolean(A=Get bRigidBodyEnabled())
   ) -> (Pose) {
    alphaBoolBlend {blendOption=Linear}
    controlRigAssetReference {blueprintRigClass=&/M_MED_FluffWoof_FaceAcc_CtrlRig.M_MED_FluffWoof_FaceAcc_CtrlRig_C}
   }
   n2 = RigidBody!(ComponentPose=Local To Component(LocalPose=@n1), bAlphaBoolEnabled:bool, OverrideWorldGravity:Vector="0,0,0", ExternalForce) -> (Pose) {
    alphaInputType=Bool
    bOverrideWorldGravity=true
    overridePhysicsAsset=&/M_MED_FluffWoof_FaceAcc_Physics
    simulationSpace=BaseBoneSpace
    alphaBoolBlend {blendOption=Linear}
    baseBoneRef {boneName=spine_05}
    componentLinearAccScale=<0.2 0.2 0.4>
    componentLinearVelScale=<0.2 0.2 0.4>
   }
   n3 = RigidBody(ComponentPose=@n2, OverrideWorldGravity="0,0,0") {
    alphaInputType=Bool
    bOverrideWorldGravity=true
    overridePhysicsAsset=&/M_MED_FluffWoof_TieFrontEnd_Physics
    simulationSpace=BaseBoneSpace
    alphaBoolBlend {blendOption=Linear}
    baseBoneRef {boneName=spine_05}
    componentLinearAccScale=<0.2 0.2 0.4>
    componentLinearVelScale=<0.2 0.2 0.4>
   }
   n4 = Anim Dynamics - Bone: dyn_scissors00(ComponentPose=@n3, Alpha=1, ExternalForce="0,0,0") {
    angularDampingOverride=1
    angularSpringConstant=20000
    bAngularSpring=true
    bOverrideAngularDamping=true
    bUseGravityOverride=true
    bUsePlanarLimit=false
    gravityScale=0.35
    simulationSpace=BoneRelative
    alphaBoolBlend {blendOption=Linear}
    boundBone {boneName=dyn_scissors00}
    componentAppliedLinearAccClamp=<3000 3000 3000>
    componentLinearAccScale=<1 1 1>
    componentLinearVelScale=<1 1 1>
    gravityOverride {z=-980}
    physicsBodyDefinitions [{
      boundBone {boneName=dyn_scissors00}
      boxExtents=<20 30 30>
      constraintSetup {angularLimitsMax=<-2 1 0>, angularLimitsMin=<-2 -1 0>, angularTarget=<1 0 0>}
      localJointOffset=<2 0 0>
     }]
    relativeSpaceBone {boneName=pelvis}
   }
   n5 = Anim Dynamics - Bone: dyn_comb00(ComponentPose=@n4, Alpha=1, ExternalForce="0,0,0") {
    angularDampingOverride=1
    angularSpringConstant=20000
    bAngularSpring=true
    bOverrideAngularDamping=true
    bUseGravityOverride=true
    gravityScale=0.35
    simulationSpace=BoneRelative
    alphaBoolBlend {blendOption=Linear}
    boundBone {boneName=dyn_comb00}
    componentAppliedLinearAccClamp=<3000 3000 3000>
    componentLinearAccScale=<1 1 1>
    componentLinearVelScale=<1 1 1>
    gravityOverride {z=-980}
    physicsBodyDefinitions [{
      boundBone {boneName=dyn_comb00}
      boxExtents=<20 30 30>
      constraintSetup {angularLimitsMax=<2 -2 1>, angularLimitsMin=<-2 -2 -1>, angularTarget=<1 0 0>}
      localJointOffset=<4 0 0>
     }]
    relativeSpaceBone {boneName=pelvis}
   }
   n6 = Get Jumping()
   n7 = "Transform (Modify) Bone - Bone: base_mount"!(ComponentPose=@n5, bAlphaBoolEnabled:bool=@n6, Translation="-1.5,-0.5,5.616", Rotation="-40,17,-4", Scale="1,1,1") -> (Pose) {
    alphaInputType=Bool
    rotationMode=BMM_Additive
    rotationSpace=BCS_BoneSpace
    translationMode=BMM_Additive
    translationSpace=BCS_BoneSpace
    alphaBoolBlend {blendInTime=0.5, blendOption=Linear, blendOutTime=0.5}
    boneToModify {boneName=base_mount}
    rotation=<rot -40 17 -4>
    translation=<-1.5 -0.5 5.616>
   }
   n8 = "Transform (Modify) Bone - Bone: base_scissors00"!(ComponentPose=@n7, bAlphaBoolEnabled:bool=@n6, Translation="0,-0.5,0", Rotation="0,0,-25", Scale="1,1,1") -> (Pose) {
    alphaInputType=Bool
    rotationMode=BMM_Additive
    rotationSpace=BCS_BoneSpace
    translationMode=BMM_Additive
    translationSpace=BCS_BoneSpace
    alphaBoolBlend {blendInTime=0.5, blendOption=Linear, blendOutTime=0.5}
    boneToModify {boneName=base_scissors00}
    rotation {roll=-25}
    translation {y=-0.5}
   }
   n9 = "Transform (Modify) Bone - Bone: base_top_strap00"!(ComponentPose=@n8, bAlphaBoolEnabled:bool=@n6, Translation="0.235,0.010,-0.197", Rotation="50.000,0.000,-0.000", Scale="1,1,1") -> (Pose) {
    alphaInputType=Bool
    rotationMode=BMM_Additive
    rotationSpace=BCS_BoneSpace
    translationMode=BMM_Additive
    translationSpace=BCS_BoneSpace
    alphaBoolBlend {blendInTime=0.5, blendOption=Linear, blendOutTime=0.5}
    boneToModify {boneName=base_top_strap00}
    rotation=<rot 50 0 0>
    translation=<0.235 0.01 -0.197>
   }
   n10 = "Transform (Modify) Bone - Bone: base_scissors01"!(ComponentPose=@n9, bAlphaBoolEnabled:bool=@n6, Translation="1,0.5,0", Rotation="0.000,10.000,-0.000", Scale="0.9,1,1") -> (Pose) {
    alphaInputType=Bool
    rotationMode=BMM_Additive
    rotationSpace=BCS_BoneSpace
    scaleMode=BMM_Additive
    scaleSpace=BCS_BoneSpace
    translationMode=BMM_Additive
    translationSpace=BCS_BoneSpace
    alphaBoolBlend {blendInTime=0.2, blendOption=Linear, blendOutTime=0.5}
    boneToModify {boneName=base_scissors01}
    rotation=<rot 0 10 0>
    scale {x=0.9}
    translation=<1 0.5>
   }
   Output Pose(Result=Component To Local(ComponentPose=@n10)) {layerGroup=None}
   n11 = Layered blend per bone(
    BasePose=Copy Pose From Mesh() {bCopyCurves=true, bUseAttachedParent=true}
    BlendPoses_0=Copy Pose From Mesh(SourceMeshComponent="Get Body Mesh "() -> ("Body Mesh ":Skeletal Mesh Component Object Reference))
   ) {layerSetup [{branchFilters [{boneName=thigh_l}, {boneName=clavicle_r}, {boneName=clavicle_l}]}]}
   Save cached pose 'Start_Cache'(Pose=@n11) {cacheName=Start_Cache}
  ]}
 EventGraph {nodes [
   Event BlueprintUpdateAnimation()
   > n1 = Sequence()
   > n2 = Gravity Override State Function DEPRECATED(CharacterPartAnimInstance=Self-Reference(), GameTimeInSeconds=GetGameTimeInSeconds(), GravityOverrideParams=Get Gravity Override Params())
   > Set Gravity Override(Gravity Override=@n2)
   Set isNotFrontEnd(execute=@n1.then_1, isNotFrontEnd:bool=NOT Boolean(A=Get bIsInFrontEnd())) -> (Output_Get:bool)
   > Set Moving2D(Moving2D:bool=OR Boolean(A=Get bIsMoving2D(), B=Get Jumping())) -> (Output_Get:bool)
   > Set ontheKnees(ontheKnees:bool=OR Boolean(A=Get bIsCrouching(), B=Get bIsDBNO())) -> (Output_Get:bool)
   Event OnBeginSkydiving()
   > Set Jumping(Jumping:bool=false) -> (Output_Get:bool)
   Event OnCharacterCustomizationLoaded()
   > n3 = Cast To PlayerPawn_Athena(Object=TryGetPawnOwner())
   > "Set Body Mesh "("Body Mesh ":Skeletal Mesh Component Object Reference=GetSkeletalMeshForPartType(self=@n3)) -> (Output_Get:Skeletal Mesh Component Object Reference)
   Event OnBeginJumping()
   > Set Jumping(Jumping=true)
   Event OnEndFalling()
   > Set Jumping(Jumping=false)
   Event OnBeginFalling()
   Set Jumping(Jumping=false)
  ]}
]
variables [
 "Body Mesh " {type=SkeletalMeshComponent}
 Gravity Override {type=Vector}
 Gravity Override Params {type=GravityOverrideParamsStruct}
 isNotFrontEnd {type=boolean}
 Jumping {type=boolean}
 Moving2D {type=boolean}
 ontheKnees {type=boolean}
]
